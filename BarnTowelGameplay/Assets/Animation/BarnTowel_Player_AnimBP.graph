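# Player Animation Blueprint for the BarnTowel flight item, implementing Item Override layers on the Fortnite player skeleton.
# ItemFullBodyOverride switches between flying and ground state machines on bIsFlying, smoothed by Dead Blending.
# The flying machine covers takeoff, hover, flight, thrust, dodges, slam, out-of-fuel and forced dive.
# It routes through state aliases and a conduit, often using inertialized transitions, and layers additive pitch/roll blendspaces onto many poses.
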AnimBlueprint /BarnTowelGameplay/Assets/Animation/BarnTowel_Player_AnimBP
target_skeleton=&Fortnite_M_Avg_Player_Skeleton
graphs [
 AnimGraph {nodes [Output Pose() {layerGroup=None}]}
 EventGraph {links [], nodes []}
 ItemAimOffset {nodes [
   Input Pose!() -> (Pose, AimOffsetAlpha:double, Yaw:double, Pitch:double) {name=InPose_AimOffset}
   Input Pose!() -> (Pose) {name=InPose_PreAimOffsets}
   Output Pose() {layerGroup=Item Override}
  ]}
 ItemFallAdditive {nodes [
   Input Pose() {name=InPoseItemFallAdditive}
   Output Pose() {layerGroup=Item Override}
  ]}
 ItemFallLandAdditive {nodes [
   Input Pose() {name=InPoseItemFallLandAdditive}
   Output Pose() {layerGroup=Item Override}
  ]}
 ItemFallLevel2Override {nodes [
   Input Pose() {name=InPose_FallLevel2}
   Output Pose() {layerGroup=Item Override}
  ]}
 ItemFinalPoseOverride {nodes [
   Input Pose() {name=InFinalPose}
   Output Pose() {layerGroup=Item Override}
  ]}
 ItemFullBodyAimOffsets {nodes [
   Input Pose!() -> (Pose, FullBody_AimOffsetAlpha:double, Aim_Yaw:double, Aim_Pitch:double) {name=InPoseFullBodyAimOffsets}
   Input Pose!() -> (Pose) {name=InPosePreFullBodyAimOffsets}
   Output Pose() {layerGroup=Item Override}
  ]}
 ItemFullBodyOverride {nodes [
   AnimGraphNode_StateMachine_0 = FlyingStateMachine() {is=title AnimGraphNode_StateMachine, bAllowConduitEntryStates=true}
   n1 = Modify Curve!(SourcePose=@AnimGraphNode_StateMachine_0, CurveValues_0=1, CurveValues_1:float=1, Alpha=1) -> (Pose) {curveNames [BothLegFK,InFlight], curveValues [0,0]}
   AnimGraphNode_StateMachine_1 = GroundStateMachine() {is=title AnimGraphNode_StateMachine, bAllowConduitEntryStates=true}
   n2 = Get bIsFlying() -> (bIsFlying:bool)
   n3 = Get FalseBlendTime() -> (FalseBlendTime:float)
   Output Pose(Result=Dead Blending!(Source:Pose Link Structure=Blend Poses by bool(BlendPose_0=@n1, BlendPose_1=@AnimGraphNode_StateMachine_1, BlendTime_0=0.015, BlendTime_1=@n3, bActiveValue=@n2) {blendType=Sinusoidal, childUpateMode=ResetChildOnActivate}) -> (Pose:Pose Link Structure) {is=title AnimGraphNode_DeadBlending, defaultBlendDuration=0.1, defaultBlendMode=Linear}) {layerGroup=Item Override}
   Save cached pose 'FullBodyCache'(Pose=Input Pose() {name=InPoseFullBody}) {cacheName=FullBodyCache}
  ]}
 FlyingStateMachine {
  entry=Into Flight
  path=ItemFullBodyOverride.AnimGraphNode_StateMachine_0.FlyingStateMachine
  ids=id
  nodes [
   ForceFall Alias() {
    is=title AnimStateAliasNode
    also {aliases [Enter Flight Loop, Exit Flight, Flight Loop, Hover, Into Flight, OutOfFuel Into, OutOfFuel Loop, Thrust Takeoff, Thrust Windup]}
   }
   Into Dodges() {
    is=title AnimStateAliasNode
    also {aliases [
      Dodge Left 1
      Dodge Left 2
      Dodge Right 1
      Dodge Right 2
      Enter Flight Loop
      Flight Loop
      ForceFall
      ForceFallLoop
      Into Flight
      Into Flight Loop
      Into Flight To Hover
      OutOfFuel Into
      OutOfFuel Loop
      Thrust Takeoff
      Thrust Windup
      Thrust Windup From Hover
      Thrusting Loop
     ]}
   }
   OutOfFuel Alias() {
    is=title AnimStateAliasNode
    also {aliases [Enter Flight Loop, Exit Flight, Flight Loop, Hover, Into Flight, Into Flight Loop, Into Flight To Hover, Thrust Takeoff, Thrust Windup]}
   }
   To Flight Loop() {is=title AnimStateAliasNode, also {aliases [Into Flight To Hover]}}
   Into Hover() {is=title AnimStateAliasNode, also {aliases [Into Flight Loop]}}
   Into Dodges Hover() {is=title AnimStateAliasNode, also {aliases [Dodge H Left 1, Dodge H Left 2, Dodge H Right 1, Dodge H Right 2, Exit Flight, Hover]}}
   Dodge Return() {
    is=title AnimStateAliasNode
    also {aliases [Dodge H Left 1, Dodge H Left 2, Dodge H Right 1, Dodge H Right 2, Dodge Left 1, Dodge Left 2, Dodge Right 1, Dodge Right 2]}
   }
   SlamIntroGlobal() {
    is=title AnimStateAliasNode
    also {aliases [Flight Loop, ForceFall, ForceFallLoop, Into Flight, Into Flight Loop, OutOfFuel Into, OutOfFuel Loop, Thrusting Loop]}
   }
   IntoThrust() {is=title AnimStateAliasNode, also {aliases [Enter Flight Loop, Flight Loop, Into Flight, Into Flight Loop, Thrust Takeoff, Thrusting Loop]}}
   IntoThrustHover() {is=title AnimStateAliasNode, also {aliases [Exit Flight, Hover, Into Flight To Hover]}}
   AnimStateConduitNode_0 = Conduit()
   AnimStateNode_0 = Thrust Windup From Hover() {is=title AnimStateNode}
   AnimStateNode_1 = Dodge H Right 1() {is=title AnimStateNode}
   AnimStateNode_10 = OutOfFuel Into() {is=title AnimStateNode}
   AnimStateNode_11 = Dodge H Left 2() {is=title AnimStateNode}
   AnimStateNode_12 = Dodge H Left 1() {is=title AnimStateNode}
   AnimStateNode_13 = Hover() {
    is=title AnimStateNode
    stateEntered {guid=#1}
    stateFullyBlended {guid=#2}
    stateLeft {guid=#3}
   }
   AnimStateNode_14 = ForceFallLoop() {is=title AnimStateNode}
   AnimStateNode_15 = Dodge H Right 2() {is=title AnimStateNode}
   AnimStateNode_16 = Into Flight Loop() {is=title AnimStateNode}
   AnimStateNode_17 = Into Flight To Hover() {is=title AnimStateNode}
   AnimStateNode_18 = Dodge Right 2() {is=title AnimStateNode}
   AnimStateNode_19 = SlamIntro() {is=title AnimStateNode}
   AnimStateNode_2 = Exit Flight() {is=title AnimStateNode}
   AnimStateNode_20 = Into Flight() {is=title AnimStateNode}
   AnimStateNode_21 = Dodge Left 1() {is=title AnimStateNode}
   AnimStateNode_22 = Dodge Left 2() {is=title AnimStateNode}
   AnimStateNode_23 = Dodge Right 1() {is=title AnimStateNode}
   AnimStateNode_3 = Thrusting Loop() {is=title AnimStateNode}
   AnimStateNode_4 = Thrust Windup() {is=title AnimStateNode}
   AnimStateNode_5 = OutOfFuel Loop() {is=title AnimStateNode}
   AnimStateNode_6 = Flight Loop() {
    is=title AnimStateNode
    stateEntered {guid=#4}
    stateFullyBlended {guid=#5}
    stateLeft {guid=#6}
   }
   AnimStateNode_7 = Enter Flight Loop() {is=title AnimStateNode}
   AnimStateNode_8 = ForceFall() {is=title AnimStateNode}
   AnimStateNode_9 = Thrust Takeoff() {is=title AnimStateNode}
   AnimStateTransitionNode_0 = AnimStateTransitionNode() {automaticRuleTriggerTime=0, crossfadeDuration=0.4, logicType=TLT_Inertialization, priorityOrder=100}
   AnimStateTransitionNode_1 = AnimStateTransitionNode()
   AnimStateTransitionNode_10 = AnimStateTransitionNode() {automaticRuleTriggerTime=0.2}
   AnimStateTransitionNode_11 = AnimStateTransitionNode() {bAutomaticRuleBasedOnSequencePlayerInState=true}
   AnimStateTransitionNode_12 = AnimStateTransitionNode()
   AnimStateTransitionNode_13 = AnimStateTransitionNode() {automaticRuleTriggerTime=0, logicType=TLT_Inertialization}
   AnimStateTransitionNode_14 = AnimStateTransitionNode()
   AnimStateTransitionNode_15 = AnimStateTransitionNode() {automaticRuleTriggerTime=0, logicType=TLT_Inertialization, priorityOrder=2}
   AnimStateTransitionNode_17 = AnimStateTransitionNode()
   AnimStateTransitionNode_18 = AnimStateTransitionNode() {automaticRuleTriggerTime=0, crossfadeDuration=0.1, logicType=TLT_Inertialization}
   AnimStateTransitionNode_19 = AnimStateTransitionNode()
   AnimStateTransitionNode_2 = AnimStateTransitionNode() {automaticRuleTriggerTime=0, bAutomaticRuleBasedOnSequencePlayerInState=true, logicType=TLT_Inertialization}
   AnimStateTransitionNode_20 = AnimStateTransitionNode() {automaticRuleTriggerTime=0.3, bAutomaticRuleBasedOnSequencePlayerInState=true}
   AnimStateTransitionNode_21 = AnimStateTransitionNode() {automaticRuleTriggerTime=0, logicType=TLT_Inertialization}
   AnimStateTransitionNode_22 = AnimStateTransitionNode()
   AnimStateTransitionNode_23 = AnimStateTransitionNode() {automaticRuleTriggerTime=0, logicType=TLT_Inertialization}
   AnimStateTransitionNode_24 = AnimStateTransitionNode() {automaticRuleTriggerTime=0, logicType=TLT_Inertialization}
   AnimStateTransitionNode_25 = AnimStateTransitionNode() {automaticRuleTriggerTime=0, logicType=TLT_Inertialization}
   AnimStateTransitionNode_26 = AnimStateTransitionNode()
   AnimStateTransitionNode_27 = AnimStateTransitionNode()
   AnimStateTransitionNode_28 = AnimStateTransitionNode() {automaticRuleTriggerTime=0, logicType=TLT_Inertialization, priorityOrder=2}
   AnimStateTransitionNode_29 = AnimStateTransitionNode()
   AnimStateTransitionNode_3 = AnimStateTransitionNode()
   AnimStateTransitionNode_30 = AnimStateTransitionNode()
   AnimStateTransitionNode_31 = AnimStateTransitionNode()
   AnimStateTransitionNode_32 = AnimStateTransitionNode()
   AnimStateTransitionNode_33 = AnimStateTransitionNode()
   AnimStateTransitionNode_34 = AnimStateTransitionNode()
   AnimStateTransitionNode_35 = AnimStateTransitionNode()
   AnimStateTransitionNode_36 = AnimStateTransitionNode() {automaticRuleTriggerTime=0.2, bAutomaticRuleBasedOnSequencePlayerInState=true}
   AnimStateTransitionNode_37 = AnimStateTransitionNode()
   AnimStateTransitionNode_38 = AnimStateTransitionNode() {
    automaticRuleTriggerTime=0
    crossfadeDuration=0.5
    logicType=TLT_Inertialization
    transitionEnd {guid=#7}
    transitionInterrupt {guid=#8}
    transitionStart {guid=#9}
   }
   AnimStateTransitionNode_39 = AnimStateTransitionNode()
   AnimStateTransitionNode_4 = AnimStateTransitionNode() {bAutomaticRuleBasedOnSequencePlayerInState=true}
   AnimStateTransitionNode_40 = AnimStateTransitionNode() {automaticRuleTriggerTime=0, bAutomaticRuleBasedOnSequencePlayerInState=true, logicType=TLT_Inertialization}
   AnimStateTransitionNode_41 = AnimStateTransitionNode() {automaticRuleTriggerTime=0, bAutomaticRuleBasedOnSequencePlayerInState=true, logicType=TLT_Inertialization}
   AnimStateTransitionNode_42 = AnimStateTransitionNode()
   AnimStateTransitionNode_43 = AnimStateTransitionNode() {automaticRuleTriggerTime=0, logicType=TLT_Inertialization}
   AnimStateTransitionNode_44 = AnimStateTransitionNode()
   AnimStateTransitionNode_45 = AnimStateTransitionNode()
   AnimStateTransitionNode_46 = AnimStateTransitionNode()
   AnimStateTransitionNode_47 = AnimStateTransitionNode()
   AnimStateTransitionNode_5 = AnimStateTransitionNode() {automaticRuleTriggerTime=0, logicType=TLT_Inertialization}
   AnimStateTransitionNode_6 = AnimStateTransitionNode() {automaticRuleTriggerTime=0, crossfadeDuration=0.4, priorityOrder=2}
   AnimStateTransitionNode_7 = AnimStateTransitionNode()
   AnimStateTransitionNode_8 = AnimStateTransitionNode() {automaticRuleTriggerTime=0, bAutomaticRuleBasedOnSequencePlayerInState=true, logicType=TLT_Inertialization}
   AnimStateTransitionNode_9 = AnimStateTransitionNode() {automaticRuleTriggerTime=0, logicType=TLT_Inertialization}
  ]
  links [
   {
    dst=Enter Flight Loop
    rule=ItemFullBodyOverride.AnimGraphNode_StateMachine_0.FlyingStateMachine.AnimStateTransitionNode_6.Transition
    src=Conduit
    via=AnimStateTransitionNode_6
   }
   {
    dst=Exit Flight
    rule=ItemFullBodyOverride.AnimGraphNode_StateMachine_0.FlyingStateMachine.AnimStateTransitionNode_15.Transition
    src=Conduit
    via=AnimStateTransitionNode_15
   }
   {
    dst=Flight Loop
    rule=ItemFullBodyOverride.AnimGraphNode_StateMachine_0.FlyingStateMachine.AnimStateTransitionNode_31.Transition
    src=Conduit
    via=AnimStateTransitionNode_31
   }
   {
    dst=ForceFall
    rule=ItemFullBodyOverride.AnimGraphNode_StateMachine_0.FlyingStateMachine.AnimStateTransitionNode_9.Transition
    src=Conduit
    via=AnimStateTransitionNode_9
   }
   {
    dst=Hover
    rule=ItemFullBodyOverride.AnimGraphNode_StateMachine_0.FlyingStateMachine.AnimStateTransitionNode_0.Transition
    src=Conduit
    via=AnimStateTransitionNode_0
   }
   {
    dst=OutOfFuel Into
    rule=ItemFullBodyOverride.AnimGraphNode_StateMachine_0.FlyingStateMachine.AnimStateTransitionNode_19.Transition
    src=Conduit
    via=AnimStateTransitionNode_19
   }
   {
    dst=Conduit
    rule=ItemFullBodyOverride.AnimGraphNode_StateMachine_0.FlyingStateMachine.AnimStateTransitionNode_3.Transition
    src=Dodge Return
    via=AnimStateTransitionNode_3
   }
   {
    dst=Conduit
    rule=ItemFullBodyOverride.AnimGraphNode_StateMachine_0.FlyingStateMachine.AnimStateTransitionNode_36.Transition
    src=Dodge Return
    via=AnimStateTransitionNode_36
   }
   {
    dst=Conduit
    rule=ItemFullBodyOverride.AnimGraphNode_StateMachine_0.FlyingStateMachine.AnimStateTransitionNode_17.Transition
    src=Enter Flight Loop
    via=AnimStateTransitionNode_17
   }
   {
    dst=Exit Flight
    rule=ItemFullBodyOverride.AnimGraphNode_StateMachine_0.FlyingStateMachine.AnimStateTransitionNode_15.Transition
    src=Enter Flight Loop
    via=AnimStateTransitionNode_43
   }
   {
    dst=Flight Loop
    rule=ItemFullBodyOverride.AnimGraphNode_StateMachine_0.FlyingStateMachine.AnimStateTransitionNode_8.Transition
    src=Enter Flight Loop
    via=AnimStateTransitionNode_8
   }
   {
    dst=Conduit
    rule=ItemFullBodyOverride.AnimGraphNode_StateMachine_0.FlyingStateMachine.AnimStateTransitionNode_10.Transition
    src=Exit Flight
    via=AnimStateTransitionNode_10
   }
   {
    dst=Conduit
    rule=ItemFullBodyOverride.AnimGraphNode_StateMachine_0.FlyingStateMachine.AnimStateTransitionNode_27.Transition
    src=Exit Flight
    via=AnimStateTransitionNode_27
   }
   {
    dst=Enter Flight Loop
    rule=ItemFullBodyOverride.AnimGraphNode_StateMachine_0.FlyingStateMachine.AnimStateTransitionNode_28.Transition
    src=Exit Flight
    via=AnimStateTransitionNode_28
   }
   {
    dst=Conduit
    rule=ItemFullBodyOverride.AnimGraphNode_StateMachine_0.FlyingStateMachine.AnimStateTransitionNode_12.Transition
    src=Flight Loop
    via=AnimStateTransitionNode_12
   }
   {
    dst=Thrust Windup
    rule=ItemFullBodyOverride.AnimGraphNode_StateMachine_0.FlyingStateMachine.AnimStateTransitionNode_16.Transition
    src=Flight Loop
    via=AnimStateTransitionNode_23
   }
   {
    dst=ForceFallLoop
    rule=ItemFullBodyOverride.AnimGraphNode_StateMachine_0.FlyingStateMachine.AnimStateTransitionNode_11.Transition
    src=ForceFall
    via=AnimStateTransitionNode_11
   }
   {
    dst=ForceFall
    rule=ItemFullBodyOverride.AnimGraphNode_StateMachine_0.FlyingStateMachine.AnimStateTransitionNode_14.Transition
    src=ForceFall Alias
    via=AnimStateTransitionNode_14
   }
   {
    dst=Conduit
    rule=ItemFullBodyOverride.AnimGraphNode_StateMachine_0.FlyingStateMachine.AnimStateTransitionNode_1.Transition
    src=Hover
    via=AnimStateTransitionNode_1
   }
   {
    dst=Conduit
    rule=ItemFullBodyOverride.AnimGraphNode_StateMachine_0.FlyingStateMachine.AnimStateTransitionNode_35.Transition
    src=Hover
    via=AnimStateTransitionNode_35
   }
   {
    dst=Dodge Left 1
    rule=ItemFullBodyOverride.AnimGraphNode_StateMachine_0.FlyingStateMachine.AnimStateTransitionNode_45.AnimationTransitionGraph_1
    src=Into Dodges
    via=AnimStateTransitionNode_45
   }
   {
    dst=Dodge Left 2
    rule=ItemFullBodyOverride.AnimGraphNode_StateMachine_0.FlyingStateMachine.AnimStateTransitionNode_29.AnimationTransitionGraph_1
    src=Into Dodges
    via=AnimStateTransitionNode_29
   }
   {
    dst=Dodge Right 1
    rule=ItemFullBodyOverride.AnimGraphNode_StateMachine_0.FlyingStateMachine.AnimStateTransitionNode_47.AnimationTransitionGraph_1
    src=Into Dodges
    via=AnimStateTransitionNode_47
   }
   {
    dst=Dodge Right 2
    rule=ItemFullBodyOverride.AnimGraphNode_StateMachine_0.FlyingStateMachine.AnimStateTransitionNode_46.AnimationTransitionGraph_1
    src=Into Dodges
    via=AnimStateTransitionNode_46
   }
   {
    dst=Dodge H Left 1
    rule=ItemFullBodyOverride.AnimGraphNode_StateMachine_0.FlyingStateMachine.AnimStateTransitionNode_34.AnimationTransitionGraph_1
    src=Into Dodges Hover
    via=AnimStateTransitionNode_34
   }
   {
    dst=Dodge H Left 2
    rule=ItemFullBodyOverride.AnimGraphNode_StateMachine_0.FlyingStateMachine.AnimStateTransitionNode_33.AnimationTransitionGraph_1
    src=Into Dodges Hover
    via=AnimStateTransitionNode_33
   }
   {
    dst=Dodge H Right 1
    rule=ItemFullBodyOverride.AnimGraphNode_StateMachine_0.FlyingStateMachine.AnimStateTransitionNode_44.AnimationTransitionGraph_1
    src=Into Dodges Hover
    via=AnimStateTransitionNode_44
   }
   {
    dst=Dodge H Right 2
    rule=ItemFullBodyOverride.AnimGraphNode_StateMachine_0.FlyingStateMachine.AnimStateTransitionNode_37.AnimationTransitionGraph_1
    src=Into Dodges Hover
    via=AnimStateTransitionNode_37
   }
   {
    dst=Into Flight Loop
    rule=ItemFullBodyOverride.AnimGraphNode_StateMachine_0.FlyingStateMachine.AnimStateTransitionNode_22.Transition
    src=Into Flight
    via=AnimStateTransitionNode_22
   }
   {
    dst=Into Flight To Hover
    rule=ItemFullBodyOverride.AnimGraphNode_StateMachine_0.FlyingStateMachine.AnimStateTransitionNode_30.Transition
    src=Into Flight
    via=AnimStateTransitionNode_30
   }
   {
    dst=Flight Loop
    rule=ItemFullBodyOverride.AnimGraphNode_StateMachine_0.FlyingStateMachine.AnimStateTransitionNode_40.Transition
    src=Into Flight Loop
    via=AnimStateTransitionNode_40
   }
   {
    dst=Hover
    rule=ItemFullBodyOverride.AnimGraphNode_StateMachine_0.FlyingStateMachine.AnimStateTransitionNode_41.Transition
    src=Into Flight To Hover
    via=AnimStateTransitionNode_41
   }
   {
    dst=Hover
    rule=ItemFullBodyOverride.AnimGraphNode_StateMachine_0.FlyingStateMachine.AnimStateTransitionNode_42.Transition
    src=Into Flight To Hover
    via=AnimStateTransitionNode_42
   }
   {
    dst=Hover
    rule=ItemFullBodyOverride.AnimGraphNode_StateMachine_0.FlyingStateMachine.AnimStateTransitionNode_39.Transition
    src=Into Hover
    via=AnimStateTransitionNode_39
   }
   {
    dst=Thrust Windup
    rule=ItemFullBodyOverride.AnimGraphNode_StateMachine_0.FlyingStateMachine.AnimStateTransitionNode_16.Transition
    src=IntoThrust
    via=AnimStateTransitionNode_25
   }
   {
    dst=Thrust Windup From Hover
    rule=ItemFullBodyOverride.AnimGraphNode_StateMachine_0.FlyingStateMachine.AnimStateTransitionNode_32.Transition
    src=IntoThrustHover
    via=AnimStateTransitionNode_32
   }
   {
    dst=OutOfFuel Into
    rule=ItemFullBodyOverride.AnimGraphNode_StateMachine_0.FlyingStateMachine.AnimStateTransitionNode_19.Transition
    src=OutOfFuel Alias
    via=AnimStateTransitionNode_26
   }
   {
    dst=OutOfFuel Loop
    rule=ItemFullBodyOverride.AnimGraphNode_StateMachine_0.FlyingStateMachine.AnimStateTransitionNode_4.Transition
    src=OutOfFuel Into
    via=AnimStateTransitionNode_4
   }
   {
    dst=Conduit
    rule=ItemFullBodyOverride.AnimGraphNode_StateMachine_0.FlyingStateMachine.AnimStateTransitionNode_20.Transition
    src=SlamIntro
    via=AnimStateTransitionNode_20
   }
   {
    dst=Conduit
    rule=ItemFullBodyOverride.AnimGraphNode_StateMachine_0.FlyingStateMachine.AnimStateTransitionNode_7.Transition
    src=SlamIntro
    via=AnimStateTransitionNode_7
   }
   {
    dst=SlamIntro
    rule=ItemFullBodyOverride.AnimGraphNode_StateMachine_0.FlyingStateMachine.AnimStateTransitionNode_18.Transition
    src=SlamIntroGlobal
    via=AnimStateTransitionNode_18
   }
   {
    dst=Thrust Windup
    rule=ItemFullBodyOverride.AnimGraphNode_StateMachine_0.FlyingStateMachine.AnimStateTransitionNode_16.Transition
    src=Thrust Takeoff
    via=AnimStateTransitionNode_24
   }
   {
    dst=Thrusting Loop
    rule=ItemFullBodyOverride.AnimGraphNode_StateMachine_0.FlyingStateMachine.AnimStateTransitionNode_2.Transition
    src=Thrust Takeoff
    via=AnimStateTransitionNode_2
   }
   {
    dst=Thrust Takeoff
    rule=ItemFullBodyOverride.AnimGraphNode_StateMachine_0.FlyingStateMachine.AnimStateTransitionNode_21.Transition
    src=Thrust Windup
    via=AnimStateTransitionNode_21
   }
   {
    dst=Thrust Takeoff
    rule=ItemFullBodyOverride.AnimGraphNode_StateMachine_0.FlyingStateMachine.AnimStateTransitionNode_5.Transition
    src=Thrust Windup From Hover
    via=AnimStateTransitionNode_5
   }
   {
    dst=Flight Loop
    rule=ItemFullBodyOverride.AnimGraphNode_StateMachine_0.FlyingStateMachine.AnimStateTransitionNode_13.Transition
    src=Thrusting Loop
    via=AnimStateTransitionNode_13
   }
   {
    dst=Flight Loop
    rule=ItemFullBodyOverride.AnimGraphNode_StateMachine_0.FlyingStateMachine.AnimStateTransitionNode_38.Transition
    src=To Flight Loop
    via=AnimStateTransitionNode_38
   }
  ]
 }
 Conduit {
  path=ItemFullBodyOverride.AnimGraphNode_StateMachine_0.FlyingStateMachine.AnimStateConduitNode_0.Conduit
  nodes [Result(bCanEnterTransition=true)]
 }
 Thrust Windup From Hover {
  path=ItemFullBodyOverride.AnimGraphNode_StateMachine_0.FlyingStateMachine.AnimStateNode_0.Thrust Windup From Hover
  nodes [
   n1 = Get AimingPitch() -> (AimingPitch:float)
   Output Animation Pose(Result=Apply Additive(
     Base=Sequence Player 'BarnTowel_Flight_Hover_To_Thrust_WindUp_Intro'() {bLoopAnimation=false, sequence=/BarnTowelGameplay/Assets/Animation/Hover/BarnTowel_Flight_Hover_To_Thrust_WindUp_Intro}
     Additive=Blendspace Player 'BarnTowel_Player_Boost_Pitch_Roll_BS'(X=0, Y=@n1) {blendSpace=&/BarnTowel_Player_Boost_Pitch_Roll_BS}
    ) {alphaBoolBlend {blendOption=Linear}})
  ]
 }
 Dodge H Right 1 {
  path=ItemFullBodyOverride.AnimGraphNode_StateMachine_0.FlyingStateMachine.AnimStateNode_1.Dodge H Right 1
  nodes [Output Animation Pose(Result=Sequence Player 'BarnTowel_Flight_Hover_Dodge_Right'() {bLoopAnimation=false, sequence=/BarnTowelGameplay/Assets/Animation/Hover/BarnTowel_Flight_Hover_Dodge_Right})]
 }
 OutOfFuel Into {
  path=ItemFullBodyOverride.AnimGraphNode_StateMachine_0.FlyingStateMachine.AnimStateNode_10.OutOfFuel Into
  nodes [
   Sequence Player 'BarnTowel_Flight_ForcedDive_Intro'() {bLoopAnimation=false, sequence=/BarnTowelGameplay/Assets/Animation/Flying/BarnTowel_Flight_ForcedDive_Intro}
   n1 = Blendspace Player 'BarnTowel_Player_Boost_Pitch_Roll_BS'(X=0, Y=Get FlyingPitch() -> (FlyingPitch:double)) {bIgnoreForRelevancyTest=true, blendSpace=&BarnTowel_Player_Boost_Pitch_Roll_BS, groupRole=AlwaysFollower}
   Output Animation Pose(Result=Apply Additive(
     Base=Sequence Player 'BarnTowel_Flight_OutOfFuel_Intro'() {bLoopAnimation=false, sequence=/BarnTowelGameplay/Assets/Animation/Flying/BarnTowel_Flight_OutOfFuel_Intro}
     Additive=@n1
    ) {alphaBoolBlend {blendOption=Linear}})
  ]
 }
 Dodge H Left 2 {
  path=ItemFullBodyOverride.AnimGraphNode_StateMachine_0.FlyingStateMachine.AnimStateNode_11.Dodge H Left 2
  nodes [Output Animation Pose(Result=Sequence Player 'BarnTowel_Flight_Hover_Dodge_Left'() {bLoopAnimation=false, sequence=/BarnTowelGameplay/Assets/Animation/Hover/BarnTowel_Flight_Hover_Dodge_Left})]
 }
 Dodge H Left 1 {
  path=ItemFullBodyOverride.AnimGraphNode_StateMachine_0.FlyingStateMachine.AnimStateNode_12.Dodge H Left 1
  nodes [Output Animation Pose(Result=Sequence Player 'BarnTowel_Flight_Hover_Dodge_Left'() {bLoopAnimation=false, sequence=&BarnTowel_Flight_Hover_Dodge_Left})]
 }
 Hover {
  path=ItemFullBodyOverride.AnimGraphNode_StateMachine_0.FlyingStateMachine.AnimStateNode_13.Hover
  nodes [
   n1 = Blendspace Player 'BarnTowel_Player_Hover_SpeedBS'(X=Get FP_StrafeLR_Hover_SpeedBS() -> (FP_StrafeLR_Hover_SpeedBS:double), Y=Get FP_Speed_Hover_SpeedBS() -> (FP_Speed_Hover_SpeedBS:double)) {blendSpace=&/BarnTowel_Player_Hover_SpeedBS}
   n2 = Blendspace Player 'BarnTowel_Player_Hover_Pitch_Roll_BS'(
    X=Get FP_StrafeLR_Hover_Pitch_Roll_BS() -> (FP_StrafeLR_Hover_Pitch_Roll_BS:double)
    Y=Get FP_Speed_Hover_Pitch_Roll_BS() -> (FP_Speed_Hover_Pitch_Roll_BS:double)
   ) {blendSpace=&/BarnTowel_Player_Hover_Pitch_Roll_BS}
   n3 = Get bIsAccelerating() -> (bIsAccelerating:bool)
   n4 = Get FPZSpeed() -> (FPZSpeed:double)
   Output Animation Pose(Result=Apply Additive(
     Base=Blendspace Player 'BarnTowel_Player_Hover_ZSpeed_BS'!(Y=@n4) -> (Pose) {blendSpace=/BarnTowelGameplay/Assets/Animation/Hover/BarnTowel_Player_Hover_ZSpeed_BS}
     Additive=Blend Poses by bool(BlendPose_0=@n1, BlendPose_1=@n2, BlendTime_0=0.5, BlendTime_1=0.5, bActiveValue=@n3) {blendType=Linear}
    ) {alphaBoolBlend {blendOption=Linear}}) {
    node {stateEntryFunction {functionName=OnStateEntry}}
    stateEntryFunction {bSelfContext=true, memberGuid=#10, memberName=OnHoverDecelStateEntry}
   }
  ]
 }
 ForceFallLoop {
  path=ItemFullBodyOverride.AnimGraphNode_StateMachine_0.FlyingStateMachine.AnimStateNode_14.ForceFallLoop
  nodes [Output Animation Pose(Result=Sequence Player 'BarnTowel_Flight_ForcedDive_Loop'() {bLoopAnimation=false, sequence=/BarnTowelGameplay/Assets/Animation/Flying/BarnTowel_Flight_ForcedDive_Loop})]
 }
 Dodge H Right 2 {
  path=ItemFullBodyOverride.AnimGraphNode_StateMachine_0.FlyingStateMachine.AnimStateNode_15.Dodge H Right 2
  nodes [Output Animation Pose(Result=Sequence Player 'BarnTowel_Flight_Hover_Dodge_Right'() {bLoopAnimation=false, sequence=&BarnTowel_Flight_Hover_Dodge_Right})]
 }
 Into Flight Loop {
  path=ItemFullBodyOverride.AnimGraphNode_StateMachine_0.FlyingStateMachine.AnimStateNode_16.Into Flight Loop
  nodes [
   n1 = Blendspace Player 'BarnTowel_Player_Boost_Pitch_Roll_BS'(X=Get FlyingRoll() -> (FlyingRoll:double), Y=Get FlyingPitch()) {blendSpace=&BarnTowel_Player_Boost_Pitch_Roll_BS}
   n2 = Sequence Player 'BarnTowel_Flight_IntoFlight_Flying_Takeoff'() {bLoopAnimation=false, sequence=/BarnTowelGameplay/Assets/Animation/Ground/BarnTowel_Flight_IntoFlight_Flying_Takeoff}
   Output Animation Pose(Result=Modify Curve!(SourcePose=Apply Additive(Base=@n2, Additive=@n1) {alphaBoolBlend {blendOption=Linear}}, CurveValues_0=1, Alpha=1) -> (Pose) {curveNames [UserCurve02], curveValues [0]})
  ]
 }
 Into Flight To Hover {
  path=ItemFullBodyOverride.AnimGraphNode_StateMachine_0.FlyingStateMachine.AnimStateNode_17.Into Flight To Hover
  nodes [Output Animation Pose(Result=Sequence Player 'BarnTowel_Flight_IntoFlight_Hover_Takeoff'() {bLoopAnimation=false, sequence=/BarnTowelGameplay/Assets/Animation/Ground/BarnTowel_Flight_IntoFlight_Hover_Takeoff}) {
    node {stateEntryFunction {functionName=OnHoverDecelStateEntry}}
    stateEntryFunction {bSelfContext=true, memberGuid=#10, memberName=OnHoverDecelStateEntry}
   }]
 }
 Dodge Right 2 {
  path=ItemFullBodyOverride.AnimGraphNode_StateMachine_0.FlyingStateMachine.AnimStateNode_18.Dodge Right 2
  nodes [Output Animation Pose(Result=Sequence Player 'BarnTowel_Flight_Flying_Dodge_Right'() {bLoopAnimation=false, sequence=/BarnTowelGameplay/Assets/Animation/Flying/BarnTowel_Flight_Flying_Dodge_Right})]
 }
 SlamIntro {
  path=ItemFullBodyOverride.AnimGraphNode_StateMachine_0.FlyingStateMachine.AnimStateNode_19.SlamIntro
  nodes [Output Animation Pose(Result=Sequence Player 'BarnTowel_Flight_Slam_Intro'() {bLoopAnimation=false, sequence=/BarnTowelGameplay/Assets/Animation/Ground/BarnTowel_Flight_Slam_Intro})]
 }
 Exit Flight {
  path=ItemFullBodyOverride.AnimGraphNode_StateMachine_0.FlyingStateMachine.AnimStateNode_2.Exit Flight
  nodes [Output Animation Pose(Result=Blendspace Player 'BS_BarnTowel_ExitBoost'(X=Get ExitBoostLeftRight() -> (ExitBoostLeftRight:double), Y=0) {bLoop=false, blendSpace=&/BS_BarnTowel_ExitBoost}) {
    node {stateEntryFunction {functionName=OnHoverDecelStateEntry}}
    stateEntryFunction {bSelfContext=true, memberGuid=#10, memberName=OnHoverDecelStateEntry}
   }]
 }
 Into Flight {
  path=ItemFullBodyOverride.AnimGraphNode_StateMachine_0.FlyingStateMachine.AnimStateNode_20.Into Flight
  nodes [Output Animation Pose(Result=Sequence Player 'BarnTowel_Flight_IntoFlight_Intro'() {bLoopAnimation=false, sequence=/BarnTowelGameplay/Assets/Animation/Ground/BarnTowel_Flight_IntoFlight_Intro}) {
    node {stateEntryFunction {functionName=OnHoverDecelStateEntry}}
    stateEntryFunction {bSelfContext=true, memberGuid=#10, memberName=OnHoverDecelStateEntry}
   }]
 }
 Dodge Left 1 {
  path=ItemFullBodyOverride.AnimGraphNode_StateMachine_0.FlyingStateMachine.AnimStateNode_21.Dodge Left 1
  nodes [Output Animation Pose(Result=Sequence Player 'BarnTowel_Flight_Flying_Dodge_Left'() {bLoopAnimation=false, sequence=/BarnTowelGameplay/Assets/Animation/Flying/BarnTowel_Flight_Flying_Dodge_Left})]
 }
 Dodge Left 2 {
  path=ItemFullBodyOverride.AnimGraphNode_StateMachine_0.FlyingStateMachine.AnimStateNode_22.Dodge Left 2
  nodes [Output Animation Pose(Result=Sequence Player 'BarnTowel_Flight_Flying_Dodge_Left'() {bLoopAnimation=false, sequence=&BarnTowel_Flight_Flying_Dodge_Left})]
 }
 Dodge Right 1 {
  path=ItemFullBodyOverride.AnimGraphNode_StateMachine_0.FlyingStateMachine.AnimStateNode_23.Dodge Right 1
  nodes [Output Animation Pose(Result=Sequence Player 'BarnTowel_Flight_Flying_Dodge_Right'() {bLoopAnimation=false, sequence=&BarnTowel_Flight_Flying_Dodge_Right})]
 }
 Thrusting Loop {
  path=ItemFullBodyOverride.AnimGraphNode_StateMachine_0.FlyingStateMachine.AnimStateNode_3.Thrusting Loop
  nodes [
   n1 = Blendspace Player 'BarnTowel_Player_Boost_Pitch_Roll_BS'(X=Get FlyingRoll(), Y=Get FlyingPitch()) {blendSpace=&BarnTowel_Player_Boost_Pitch_Roll_BS}
   n2 = Sequence Player 'BarnTowel_Flight_Thrust_Loop'() {sequence=/BarnTowelGameplay/Assets/Animation/Flying/BarnTowel_Flight_Thrust_Loop}
   Output Animation Pose(Result=Modify Curve(SourcePose=Apply Additive(Base=@n2, Additive=@n1) {alphaBoolBlend {blendOption=Linear}}, CurveValues_0=1, Alpha=1) {curveNames [UserCurve02], curveValues [0]})
  ]
 }
 Thrust Windup {
  path=ItemFullBodyOverride.AnimGraphNode_StateMachine_0.FlyingStateMachine.AnimStateNode_4.Thrust Windup
  nodes [
   n1 = Get AimingPitch()
   Output Animation Pose(Result=Apply Additive(
     Base=Sequence Player 'BarnTowel_Flight_Flying_To_Thrust_WindUp_Intro'() {bLoopAnimation=false, sequence=/BarnTowelGameplay/Assets/Animation/Flying/BarnTowel_Flight_Flying_To_Thrust_WindUp_Intro}
     Additive=Blendspace Player 'BarnTowel_Player_Boost_Pitch_Roll_BS'(X=0, Y=@n1) {blendSpace=&BarnTowel_Player_Boost_Pitch_Roll_BS}
    ) {alphaBoolBlend {blendOption=Linear}})
  ]
 }
 OutOfFuel Loop {
  path=ItemFullBodyOverride.AnimGraphNode_StateMachine_0.FlyingStateMachine.AnimStateNode_5.OutOfFuel Loop
  nodes [
   n1 = Get FlyingPitch()
   n2 = Get FlyingRoll()
   Output Animation Pose(Result=Apply Additive(
     Base=Sequence Player 'BarnTowel_Flight_OutOfFuel_Loop'() {bLoopAnimation=false, sequence=/BarnTowelGameplay/Assets/Animation/Flying/BarnTowel_Flight_OutOfFuel_Loop}
     Additive=Blendspace Player 'BarnTowel_Player_Boost_Pitch_Roll_BS'(X=@n2, Y=@n1) {bIgnoreForRelevancyTest=true, blendSpace=&BarnTowel_Player_Boost_Pitch_Roll_BS}
    ) {alphaBoolBlend {blendOption=Linear}})
  ]
 }
 Flight Loop {
  path=ItemFullBodyOverride.AnimGraphNode_StateMachine_0.FlyingStateMachine.AnimStateNode_6.Flight Loop
  nodes [
   n1 = Blendspace Player 'BarnTowel_Player_Boost_Pitch_Roll_BS'(X=Get FlyingRoll(), Y=Get FlyingPitch()) {blendSpace=&BarnTowel_Player_Boost_Pitch_Roll_BS}
   n2 = Sequence Player 'BarnTowel_Flight_Idle_Loop'() {sequence=/BarnTowelGameplay/Assets/Animation/Flying/BarnTowel_Flight_Idle_Loop}
   Output Animation Pose(Result=Modify Curve(SourcePose=Apply Additive(Base=@n2, Additive=@n1) {alphaBoolBlend {blendOption=Linear}}, CurveValues_0=1, Alpha=1) {curveNames [UserCurve02], curveValues [0]})
  ]
 }
 Enter Flight Loop {
  path=ItemFullBodyOverride.AnimGraphNode_StateMachine_0.FlyingStateMachine.AnimStateNode_7.Enter Flight Loop
  nodes [
   n1 = Get YawOffset() -> (YawOffset:float)
   n2 = Get FlyingPitch()
   n3 = Get FlyingRoll()
   Output Animation Pose(Result=Apply Additive(
     Base=Blendspace Player 'BS_BarnTowel_IntoBoost'(X=@n1, Y=0) {bLoop=false, blendSpace=&/BS_BarnTowel_IntoBoost}
     Additive=Blendspace Player 'BarnTowel_Player_Boost_Pitch_Roll_BS'(X=@n3, Y=@n2) {blendSpace=&BarnTowel_Player_Boost_Pitch_Roll_BS}
    ) {alphaBoolBlend {blendOption=Linear}})
  ]
 }
 ForceFall {
  path=ItemFullBodyOverride.AnimGraphNode_StateMachine_0.FlyingStateMachine.AnimStateNode_8.ForceFall
  nodes [Output Animation Pose(Result=Sequence Player 'BarnTowel_Flight_ForcedDive_Intro'() {bLoopAnimation=false, sequence=&BarnTowel_Flight_ForcedDive_Intro})]
 }
 Thrust Takeoff {
  path=ItemFullBodyOverride.AnimGraphNode_StateMachine_0.FlyingStateMachine.AnimStateNode_9.Thrust Takeoff
  nodes [
   n1 = Get FlyingPitch()
   n2 = Get FlyingRoll()
   Output Animation Pose(Result=Apply Additive(
     Base=Sequence Player 'BarnTowel_Flight_Flying_To_Thrust_Windup_Takeoff'() {bLoopAnimation=false, sequence=/BarnTowelGameplay/Assets/Animation/Flying/BarnTowel_Flight_Flying_To_Thrust_Windup_Takeoff}
     Additive=Blendspace Player 'BarnTowel_Player_Boost_Pitch_Roll_BS'(X=@n2, Y=@n1) {blendSpace=&BarnTowel_Player_Boost_Pitch_Roll_BS}
    ) {alphaBoolBlend {blendOption=Linear}})
  ]
 }
 Transition {
  path=ItemFullBodyOverride.AnimGraphNode_StateMachine_0.FlyingStateMachine.AnimStateTransitionNode_0.Transition
  nodes [Result(bCanEnterTransition=true)]
 }
 Transition {
  path=ItemFullBodyOverride.AnimGraphNode_StateMachine_0.FlyingStateMachine.AnimStateTransitionNode_1.Transition
  nodes [Result(bCanEnterTransition=true)]
 }
 Transition {
  path=ItemFullBodyOverride.AnimGraphNode_StateMachine_0.FlyingStateMachine.AnimStateTransitionNode_10.Transition
  nodes [Result(bCanEnterTransition=Get bFP_ExitFlightToConduit() -> (bFP_ExitFlightToConduit:bool))]
 }
 Transition {path=ItemFullBodyOverride.AnimGraphNode_StateMachine_0.FlyingStateMachine.AnimStateTransitionNode_11.Transition, nodes [Result()]}
 Transition {
  path=ItemFullBodyOverride.AnimGraphNode_StateMachine_0.FlyingStateMachine.AnimStateTransitionNode_12.Transition
  nodes [Result(bCanEnterTransition=Get bIsHovering() -> (bIsHovering:bool))]
 }
 Transition {
  path=ItemFullBodyOverride.AnimGraphNode_StateMachine_0.FlyingStateMachine.AnimStateTransitionNode_13.Transition
  nodes [Result(bCanEnterTransition=NOT Boolean(A=Get bThrusting() -> (bThrusting:bool)))]
 }
 Transition {
  path=ItemFullBodyOverride.AnimGraphNode_StateMachine_0.FlyingStateMachine.AnimStateTransitionNode_14.Transition
  nodes [Result(bCanEnterTransition=Get bInForceFall() -> (bInForceFall:bool))]
 }
 Transition {
  path=ItemFullBodyOverride.AnimGraphNode_StateMachine_0.FlyingStateMachine.AnimStateTransitionNode_15.Transition
  nodes [Result(bCanEnterTransition=Get bFP_ExitFlight() -> (bFP_ExitFlight:bool))]
 }
 Transition {
  path=ItemFullBodyOverride.AnimGraphNode_StateMachine_0.FlyingStateMachine.AnimStateTransitionNode_16.Transition
  nodes [Result(bCanEnterTransition=Get bWindingUpThrust() -> (bWindingUpThrust:bool))]
 }
 Transition {
  path=ItemFullBodyOverride.AnimGraphNode_StateMachine_0.FlyingStateMachine.AnimStateTransitionNode_17.Transition
  nodes [Result(bCanEnterTransition=Get bFP_EnterFlightLoopToConduit() -> (bFP_EnterFlightLoopToConduit:bool))]
 }
 Transition {
  path=ItemFullBodyOverride.AnimGraphNode_StateMachine_0.FlyingStateMachine.AnimStateTransitionNode_18.Transition
  nodes [Result(bCanEnterTransition=Get bCanEnterSlam() -> (bCanEnterSlam:bool))]
 }
 Transition {
  path=ItemFullBodyOverride.AnimGraphNode_StateMachine_0.FlyingStateMachine.AnimStateTransitionNode_19.Transition
  nodes [Result(bCanEnterTransition=Get bNoFuel() -> (bNoFuel:bool))]
 }
 Transition {path=ItemFullBodyOverride.AnimGraphNode_StateMachine_0.FlyingStateMachine.AnimStateTransitionNode_2.Transition, nodes [Result()]}
 Transition {path=ItemFullBodyOverride.AnimGraphNode_StateMachine_0.FlyingStateMachine.AnimStateTransitionNode_20.Transition, nodes [Result()]}
 Transition {
  path=ItemFullBodyOverride.AnimGraphNode_StateMachine_0.FlyingStateMachine.AnimStateTransitionNode_21.Transition
  nodes [Result(bCanEnterTransition=Get bThrusting())]
 }
 Transition {
  path=ItemFullBodyOverride.AnimGraphNode_StateMachine_0.FlyingStateMachine.AnimStateTransitionNode_22.Transition
  nodes [Result(bCanEnterTransition=Get bFP_IntoFlightToIntoFlightLoop() -> (bFP_IntoFlightToIntoFlightLoop:bool))]
 }
 Transition {
  path=ItemFullBodyOverride.AnimGraphNode_StateMachine_0.FlyingStateMachine.AnimStateTransitionNode_27.Transition
  nodes [Result(bCanEnterTransition=Get bFP_ExitFlightToConduit2() -> (bFP_ExitFlightToConduit2:bool))]
 }
 Transition {
  path=ItemFullBodyOverride.AnimGraphNode_StateMachine_0.FlyingStateMachine.AnimStateTransitionNode_28.Transition
  nodes [Result(bCanEnterTransition=Get bFP_ExitFlightToEnterFlightLoop() -> (bFP_ExitFlightToEnterFlightLoop:bool))]
 }
 AnimationTransitionGraph_1 {
  path=ItemFullBodyOverride.AnimGraphNode_StateMachine_0.FlyingStateMachine.AnimStateTransitionNode_29.AnimationTransitionGraph_1
  nodes [Result(bCanEnterTransition=Get bDodgeLeftInto2() -> (bDodgeLeftInto2:bool))]
 }
 Transition {
  path=ItemFullBodyOverride.AnimGraphNode_StateMachine_0.FlyingStateMachine.AnimStateTransitionNode_3.Transition
  nodes [Result(bCanEnterTransition=Get bFP_DodgeReturnToConduit() -> (bFP_DodgeReturnToConduit:bool))]
 }
 Transition {
  path=ItemFullBodyOverride.AnimGraphNode_StateMachine_0.FlyingStateMachine.AnimStateTransitionNode_30.Transition
  nodes [Result(bCanEnterTransition=Get bFP_IntoFlightToIntoFlightToHover() -> (bFP_IntoFlightToIntoFlightToHover:bool))]
 }
 Transition {
  path=ItemFullBodyOverride.AnimGraphNode_StateMachine_0.FlyingStateMachine.AnimStateTransitionNode_31.Transition
  nodes [Result(bCanEnterTransition=Get bFP_ConduitToFlightLoop() -> (bFP_ConduitToFlightLoop:bool))]
 }
 Transition {
  path=ItemFullBodyOverride.AnimGraphNode_StateMachine_0.FlyingStateMachine.AnimStateTransitionNode_32.Transition
  nodes [Result(bCanEnterTransition=Get bWindingUpThrust())]
 }
 AnimationTransitionGraph_1 {
  path=ItemFullBodyOverride.AnimGraphNode_StateMachine_0.FlyingStateMachine.AnimStateTransitionNode_33.AnimationTransitionGraph_1
  nodes [Result(bCanEnterTransition=Get bDodgeLeftInto2())]
 }
 AnimationTransitionGraph_1 {
  path=ItemFullBodyOverride.AnimGraphNode_StateMachine_0.FlyingStateMachine.AnimStateTransitionNode_34.AnimationTransitionGraph_1
  nodes [Result(bCanEnterTransition=Get bDodgeLeftInto1() -> (bDodgeLeftInto1:bool))]
 }
 Transition {
  path=ItemFullBodyOverride.AnimGraphNode_StateMachine_0.FlyingStateMachine.AnimStateTransitionNode_35.Transition
  nodes [Result(bCanEnterTransition=Get bWindingUpThrust())]
 }
 Transition {path=ItemFullBodyOverride.AnimGraphNode_StateMachine_0.FlyingStateMachine.AnimStateTransitionNode_36.Transition, nodes [Result()]}
 AnimationTransitionGraph_1 {
  path=ItemFullBodyOverride.AnimGraphNode_StateMachine_0.FlyingStateMachine.AnimStateTransitionNode_37.AnimationTransitionGraph_1
  nodes [Result(bCanEnterTransition=Get bDodgeRightInto2() -> (bDodgeRightInto2:bool))]
 }
 Transition {
  path=ItemFullBodyOverride.AnimGraphNode_StateMachine_0.FlyingStateMachine.AnimStateTransitionNode_38.Transition
  nodes [Result(bCanEnterTransition=NOT Boolean(A=Get bIsHovering()))]
 }
 Transition {
  path=ItemFullBodyOverride.AnimGraphNode_StateMachine_0.FlyingStateMachine.AnimStateTransitionNode_39.Transition
  nodes [Result(bCanEnterTransition=Get bIsHovering())]
 }
 Transition {path=ItemFullBodyOverride.AnimGraphNode_StateMachine_0.FlyingStateMachine.AnimStateTransitionNode_4.Transition, nodes [Result()]}
 Transition {path=ItemFullBodyOverride.AnimGraphNode_StateMachine_0.FlyingStateMachine.AnimStateTransitionNode_40.Transition, nodes [Result()]}
 Transition {path=ItemFullBodyOverride.AnimGraphNode_StateMachine_0.FlyingStateMachine.AnimStateTransitionNode_41.Transition, nodes [Result()]}
 Transition {
  path=ItemFullBodyOverride.AnimGraphNode_StateMachine_0.FlyingStateMachine.AnimStateTransitionNode_42.Transition
  nodes [Result(bCanEnterTransition=Get bIsAccelerating())]
 }
 AnimationTransitionGraph_1 {
  path=ItemFullBodyOverride.AnimGraphNode_StateMachine_0.FlyingStateMachine.AnimStateTransitionNode_44.AnimationTransitionGraph_1
  nodes [Result(bCanEnterTransition=Get bDodgeRightInto1() -> (bDodgeRightInto1:bool))]
 }
 AnimationTransitionGraph_1 {
  path=ItemFullBodyOverride.AnimGraphNode_StateMachine_0.FlyingStateMachine.AnimStateTransitionNode_45.AnimationTransitionGraph_1
  nodes [Result(bCanEnterTransition=Get bDodgeLeftInto1())]
 }
 AnimationTransitionGraph_1 {
  path=ItemFullBodyOverride.AnimGraphNode_StateMachine_0.FlyingStateMachine.AnimStateTransitionNode_46.AnimationTransitionGraph_1
  nodes [Result(bCanEnterTransition=Get bDodgeRightInto2())]
 }
 AnimationTransitionGraph_1 {
  path=ItemFullBodyOverride.AnimGraphNode_StateMachine_0.FlyingStateMachine.AnimStateTransitionNode_47.AnimationTransitionGraph_1
  nodes [Result(bCanEnterTransition=Get bDodgeRightInto1())]
 }
 Transition {
  path=ItemFullBodyOverride.AnimGraphNode_StateMachine_0.FlyingStateMachine.AnimStateTransitionNode_5.Transition
  nodes [Result(bCanEnterTransition=Get bThrusting())]
 }
 Transition {
  path=ItemFullBodyOverride.AnimGraphNode_StateMachine_0.FlyingStateMachine.AnimStateTransitionNode_6.Transition
  nodes [Result(bCanEnterTransition=Get bFP_NotHovering() -> (bFP_NotHovering:bool))]
 }
 Transition {
  path=ItemFullBodyOverride.AnimGraphNode_StateMachine_0.FlyingStateMachine.AnimStateTransitionNode_7.Transition
  nodes [Result(bCanEnterTransition=Get bFP_SlamIntroToConduit() -> (bFP_SlamIntroToConduit:bool))]
 }
 Transition {path=ItemFullBodyOverride.AnimGraphNode_StateMachine_0.FlyingStateMachine.AnimStateTransitionNode_8.Transition, nodes [Result()]}
 Transition {
  path=ItemFullBodyOverride.AnimGraphNode_StateMachine_0.FlyingStateMachine.AnimStateTransitionNode_9.Transition
  nodes [Result(bCanEnterTransition=Get bInForceFall())]
 }
 GroundStateMachine {
  entry=Ground Navigation
  path=ItemFullBodyOverride.AnimGraphNode_StateMachine_1.GroundStateMachine
  ids=id
  nodes [
   AnimStateNode_1 = Ground Navigation() {
    is=title AnimStateNode
    stateEntered {guid=#11}
    stateFullyBlended {guid=#12}
    stateLeft {guid=#13}
   }
   AnimStateNode_4 = Soft Land() {
    is=title AnimStateNode
    stateEntered {guid=#14}
    stateFullyBlended {guid=#15}
    stateLeft {guid=#16}
   }
   AnimStateTransitionNode_5 = AnimStateTransitionNode()
   AnimStateTransitionNode_7 = AnimStateTransitionNode()
  ]
  links [
   {
    dst=Soft Land
    rule=ItemFullBodyOverride.AnimGraphNode_StateMachine_1.GroundStateMachine.AnimStateTransitionNode_5.Transition
    src=Ground Navigation
    via=AnimStateTransitionNode_5
   }
   {
    dst=Ground Navigation
    rule=ItemFullBodyOverride.AnimGraphNode_StateMachine_1.GroundStateMachine.AnimStateTransitionNode_7.Transition
    src=Soft Land
    via=AnimStateTransitionNode_7
   }
  ]
 }
 Ground Navigation {
  path=ItemFullBodyOverride.AnimGraphNode_StateMachine_1.GroundStateMachine.AnimStateNode_1.Ground Navigation
  nodes [Output Animation Pose(Result=Use cached pose 'FullBodyCache'())]
 }
 Soft Land {
  path=ItemFullBodyOverride.AnimGraphNode_StateMachine_1.GroundStateMachine.AnimStateNode_4.Soft Land
  nodes [
   n1 = Get Cur_BlendIn() -> (Cur_BlendIn:float)
   n2 = Layered blend per bone(
    BasePose=Layered blend per bone(
     BasePose=Sequence Player 'BarnTowel_Flight_Land_Without_Thrusting'!(StartPosition:float=@n1) -> (Pose) {bLoopAnimation=false, sequence=/BarnTowelGameplay/Assets/Animation/Ground/BarnTowel_Flight_Land_Without_Thrusting}
     BlendPoses_0=Use cached pose 'FullBodyCache'()
     BlendWeights_0=Get Cur_foot_r() -> (Cur_foot_r:float)
    ) {layerSetup [{branchFilters [{boneName=ik_foot_r}, {boneName=thigh_r}]}]}
    BlendPoses_0=Use cached pose 'FullBodyCache'()
    BlendWeights_0=Get Cur_foot_l() -> (Cur_foot_l:float)
   ) {layerSetup [{branchFilters [{boneName=ik_foot_root}, {boneName=thigh_l}, {blendDepth=3, boneName=pelvis}]}]}
   n3 = Layered blend per bone(BasePose=@n2, BlendPoses_0=Use cached pose 'FullBodyCache'(), BlendWeights_0=Get Cur_BlendOut() -> (Cur_BlendOut:float)) {
    blendMode=BlendMask
    blendMasks [&Fortnite_M_Avg_Player_Skeleton:LowerBodyPartialUpperBody]
    layerSetup []
   }
   Output Animation Pose(Result=@n3)
  ]
 }
 Transition {
  path=ItemFullBodyOverride.AnimGraphNode_StateMachine_1.GroundStateMachine.AnimStateTransitionNode_5.Transition
  nodes [Result(bCanEnterTransition=Get bLowFrictionLand() -> (bLowFrictionLand:bool))]
 }
 Transition {
  path=ItemFullBodyOverride.AnimGraphNode_StateMachine_1.GroundStateMachine.AnimStateTransitionNode_7.Transition
  nodes [Result(bCanEnterTransition=Get bExitSoftLand() -> (bExitSoftLand:bool))]
 }
 ItemIdleOverride {nodes [
   Input Pose() {name=InPose_Idle}
   Output Pose() {layerGroup=Item Override}
  ]}
 ItemJetPackJumpAdditive {nodes [Output Pose() {layerGroup=Item Override}]}
 ItemJetPackStartAdditive {nodes [Output Pose() {layerGroup=Item Override}]}
 ItemJumpLoopAdditive {nodes [
   Input Pose() {name=InPoseItemJumpLoopAdditive}
   Output Pose() {layerGroup=Item Override}
  ]}
 ItemJumpOverride {nodes [
   Input Pose() {name=InPoseJump}
   Output Pose() {layerGroup=Item Override}
  ]}
 ItemJumpUpAdditive {nodes [
   Input Pose() {name=InPoseItemJumpUpAdditive}
   Output Pose() {layerGroup=Item Override}
  ]}
 ItemPreAimOffsets {nodes [
   Input Pose() {name=InPosePreAimOffsets}
   Output Pose() {layerGroup=Item Override}
  ]}
 ItemPreIK {nodes [
   Input Pose() {name=InPose_PreIK}
   Output Pose() {layerGroup=Item Override}
  ]}
 ItemRightHandIK {nodes [
   Input Pose() {name=InPosePreRightHandIK}
   Input Pose() {name=InPosePostRightHandIK}
   Output Pose() {layerGroup=Item Override}
  ]}
 ItemSkeletalControl {nodes [
   Input Pose()
   Output Pose() {layerGroup=Item Override}
  ]}
 ItemSwimJumpFallLoop {nodes [Output Pose() {layerGroup=Item Override}]}
 ItemSwimJumpSurfaceEnd {nodes [Output Pose() {layerGroup=Item Override}]}
 ItemSwimJumpSurfaceLoop {nodes [Output Pose() {layerGroup=Item Override}]}
 ItemUpperBody {nodes [
   Input Pose!() -> (Pose, InputParam:Fort Anim Input Adjusted Aim Structure) {name=InPoseUpperBody}
   Output Pose() {layerGroup=Item Override}
  ]}
 ItemZipLineStartAdditive {nodes [Output Pose() {layerGroup=Item Override}]}
 OnHoverDecelStateEntry {nodes [
   OnHoverDecelStateEntry!() -> (then, Context:"Anim Update Context Structure (by ref)", Node:"Anim Node Reference Structure (by ref)") {is=title K2Node_FunctionEntry}
   > Set RotationState(RotationState:float=0) -> (Output_Get:float)
   > Return Node()
  ]}
 OnHoverStateEntry {nodes [
   OnHoverStateEntry!() -> (then, Context:"Anim Update Context Structure (by ref)", Node:"Anim Node Reference Structure (by ref)") {is=title K2Node_FunctionEntry}
   > Set RotationState(RotationState=0)
   > Return Node()
  ]}
]
variables []
