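# Translucent, unlit, two-sided material for Niagara mesh particles, drawing the BarnTowel air dive wave effect.
# Color blends orange to blue along a UV-based gradient, tinted by particle color.
# Opacity combines fast-panning smoke textures, distorted by scrolling noise, with a second noise layer.
# A soft sphere mask along V, view-angle fading and particle alpha shape the wave's edges and visibility.
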Material /BarnTowelGameplay/Assets/FX/M_BarnTowel_AirDiveWave
graph {
 properties {
  bAutomaticallySetUsageInEditor=false
  bUsedWithNiagaraMeshParticles=true
  blendMode=BLEND_Translucent
  previewMesh=&/SM_BarnTowel_AirDiveWave
  shadingModel=MSM_Unlit
  twoSided=true
  usedShadingModels=Unlit
 }
 nodes [
  n1 = Constant3Vector() {constant=<rgba 3 0.5 0 1>}
  n2 = Constant3Vector() {constant=<rgba 1.1198 1.6291 5 1>}
  n3 = ComponentMask(None=TextureCoordinate()) {g=true}
  n4 = Saturate(None=Multiply(A=Power(Base=@n3)) {constB=2})
  Multiply(A=ParticleColor().RGB, B=Multiply(A=LinearInterpolate(A=@n1, B=@n2, Alpha=@n4)))
  n5 = TextureSample(UVs=Panner(Coordinate=TextureCoordinate() {uTiling=3, vTiling=0.2}) {speedY=-2}) {samplerType=SAMPLERTYPE_Grayscale, texture=&LowResBlurredNoise}
  n6 = Panner(Coordinate=Add(A=TextureCoordinate() {vTiling=0.7}, B=Multiply(A=@n5.R) {constB=0.2}) {constB=0}) {speedY=-3}
  n7 = TextureSample(UVs=@n6) {
   samplerSource=SSM_Wrap_WorldGroupSettings
   samplerType=SAMPLERTYPE_Grayscale
   texture=/Game/Effects/Fort_Effects/Textures/Smoke/T_Smoke_Trail_Soft_NoClamp
  }
  n8 = Multiply(A=@n7.R, B=@n7.R)
  n9 = Max(A=@n8, B=TextureSample(UVs=@n6) {texture=/Game/Effects/Fort_Effects/Textures/Smoke/T_Smoke_Trail_Soft_2}.B)
  n10 = TextureSample(UVs=Panner(Coordinate=TextureCoordinate() {vTiling=0.5}) {speedY=-2}) {samplerType=SAMPLERTYPE_Grayscale, texture=&LowResBlurredNoise}
  n11 = Power(Base=SphereMask(A=TextureCoordinate() {uTiling=0}, B=Constant2Vector() {g=0.4}) {attenuationRadius=0.5, hardnessPercent=0}) {constExponent=1.5}
  n12 = MaterialFunctionCall("FadeSensitivity (optional)"=Constant() {r=2}) {materialFunction=&FadeBasedOnViewAngle}
  Multiply(A=ParticleColor().A, B=Multiply(A=Multiply(A=@n9, B=@n10.R), B=Multiply(A=@n11, B=@n12.Result)))
  Power(Base=@n8)
  Saturate(None=Multiply(A=@n12.Result) {constB=3})
 ]
}
