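# Additive unlit Niagara sprite material, named as a camera-lens screen flash.
# It builds a soft radial mask from subtracted sphere masks, with Dynamic Param1 scaling the inner cutout and another mask clearing the center, multiplied by particle alpha.
# Scene color sampled at sprite UVs is tinted by particle color.
# A shading-path switch replaces scene color with zero on forward and mobile paths.
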Material /BarnTowelGameplay/Assets/FX/M_CameraLens_ScreenFlash
graph {
 properties {bUsedWithNiagaraSprites=true, blendMode=BLEND_Additive, pixelDepthOffsetMode=PDOM_Legacy, shadingModel=MSM_Unlit, usedShadingModels=Unlit}
 nodes [
  MaterialFunctionCall() {materialFunction=&DebugTimeSine}
  n1 = ParticleColor()
  n2 = DynamicParameter() {defaultValue=<rgba 0 1 1 1>}
  n3 = Subtract(
   A=SphereMask(A=TextureCoordinate(), B=Constant() {r=0.5}) {attenuationRadius=0.5, hardnessPercent=0}
   B=SphereMask(A=TextureCoordinate(), B=Constant() {r=0.5}, Radius=LinearInterpolate(Alpha=@n2.Param1) {constB=0.5}) {attenuationRadius=0.3, hardnessPercent=30}
  )
  n4 = SphereMask(A=TextureCoordinate(), B=Constant() {r=0.5}) {attenuationRadius=0.4, hardnessPercent=0}
  Multiply(A=@n1.A, B=Multiply(A=Saturate(None=@n3), B=Subtract(B=@n4)))
  n5 = SceneColor!(UVs=TextureCoordinate())
  n6 = Constant()
  Multiply(A=ShadingPathSwitch(Default=@n5.RGB, Deferred=@n5.RGB, Forward=@n6, Mobile=@n6), B=@n1.RGB)
 ]
}
